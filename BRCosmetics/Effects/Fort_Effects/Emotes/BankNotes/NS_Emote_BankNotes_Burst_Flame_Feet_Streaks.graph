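# Niagara system for the BankNotes emote that bursts a few short-lived purple flame streak sprites, named for the character's feet.
# Its single Streaks_Fat emitter places particles at a system-level skeletal mesh sample, then launches them in an upward cone with curve-driven drag.
# Sprites use SubUV frames, stretch and fade over life, and drive material animation through curve-fed dynamic parameters.
# User parameters control global sprite scale and camera offset.
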NiagaraSystem /BRCosmetics/Effects/Fort_Effects/Emotes/BankNotes/NS_Emote_BankNotes_Burst_Flame_Feet_Streaks
dependencies {
 usedDataInterfaces [
  &NiagaraDataInterfaceMeshRendererInfo
  &NiagaraDataInterfaceCurve
  &NiagaraDataInterfaceVector2DCurve
  &NiagaraDataInterfaceColorCurve
  &NiagaraDataInterfaceSkeletalMesh
 ]
 usedDynamicInputs []
 usedModules [&SystemState, &SampleSkeletalMesh]
}
emitters {Streaks_Fat {
  emitterUpdateScript [
   &EmitterState(Life Cycle Mode=::NewEnumerator1, Loop Behavior=::NewEnumerator1, Loop Delay=0.0353, Loop Duration=1, UseLoopDelay=true)
   &SpawnBurst_Instantaneous(
    Spawn Count=&RandomRangeInt(Minimum=1, Maximum=2)
    Use Spawn Probability=false
    UseLoopCountLimit=false
   )
   &SpawnBurst_Instantaneous(
    Spawn Count=&RandomRangeInt(Minimum=1, Maximum=2)
    Spawn Group=1
    Use Spawn Probability=false
    UseLoopCountLimit=false
   ) {module=SpawnBurst_Instantaneous001}
  ]
  particleSpawnScript [
   &InitializeParticle(
    Color=<rgba 0.7606 0 0.9531 1.5>
    Lifetime Max=0.4
    Lifetime Min=0.25
    Lifetime Mode=::NewEnumerator1
    Sprite Size=&Multiply_Vector2D_ByFloat(Float=$User.Global Scale, Vector2D=&UniformRangedVector2D(Minimum=<100 50>, Maximum=<150 150>))
    Sprite Size Mode=::NewEnumerator4
    UsePositionOffset=false
    Write Material Random=true
   )
   /SetVariables(Particles.Position=$System.SampleSkeletalMesh.SampledPosition) {isSetParameters=true}
   /Niagara/Modules/Spawn/Location/SphereLocation(
    MaskSpawnGroup=false
    Module.Override Randomness=false
    Sphere Origin:NiagaraPosition=$Particles.SkeletalMeshLocation.SampledPosition
    Sphere Radius=2.5
    Surface Only=true
   ) {enabled=false}
   &AddVelocity(
    Cone Angle=120
    Cone Axis=<0 0 1>
    Module.UseConeFalloff=false
    Module.UseVelDistribution=false
    Override Seed=false
    Velocity Mode=::NewEnumerator2
   )
   &AddVelocityFromPoint(
    Module.Use Falloff Distance=false
    Origin Offset=<0 0 -5>
    Velocity Strength=&UniformRangedFloat(Minimum=2, Maximum=15)
   ) {enabled=false}
   &SubUVAnimation(End Frame=3, SubUV Animation Mode:ENiagara_SubUVLookupMode=::NewEnumerator2)
   &CameraOffset(Camera Offset Amount=$User.uCameraOffset)
  ]
  particleUpdateScript [
   &ParticleState()
   &Drag(
    Drag=&FloatFromCurve(FloatCurve=DataInterface() {propertyValues {Curve {keys [<at 0.2 1>,<at 0.3 0>]}}}, Scale Curve=15)
    Use Linear Drag=true
    Use Rotational Drag=false
   )
   &GravityForce(Gravity=&Multiply_VectorByFloat(
     Vector=<0 0 -2.2>
     Float=&FloatFromCurve(
      FloatCurve=DataInterface() {propertyValues {Curve {keys [{arriveTangent=3.1357, interpMode=RCIM_Cubic, leaveTangent=3.1357, tangentMode=RCTM_User, time=0.1, value=0}, <at 0.6 1>]}}}
      CurveIndex=$Emitter.Age
     )
    )) {enabled=false}
   &SolveForcesAndVelocity(Clamp Velocity=false, Limit Acceleration=false)
   &ScaleSpriteSize(Scale Factor=&Vector2DFromCurve(
     Vector2Curve=DataInterface() {propertyValues {
       XCurve {keys [
         <at 0 0.625 RCIM_Cubic>
         {arriveTangent=0.6, interpMode=RCIM_Cubic, leaveTangent=0.6, time=0.4, value=0.75}
         {arriveTangent=0.8884, interpMode=RCIM_Cubic, leaveTangent=0.8884, tangentMode=RCTM_User, time=1, value=1.225}
        ]}
       YCurve {keys [
         {arriveTangent=45.128, interpMode=RCIM_Cubic, leaveTangent=45.128, tangentMode=RCTM_User, time=0, value=0.1}
         <at 0.1 1.625>
         <at 1 1.85 RCIM_Cubic>
        ]}
      }}
     ScaleCurve=<1.5 2.5>
    ))
   &ScaleColor(
    Linear Color Curve=DataInterface() {propertyValues {
      AlphaCurve {keys [
        {arriveTangent=-0.3577, interpMode=RCIM_Cubic, leaveTangent=-0.3577, tangentMode=RCTM_User, time=0.08, value=1}
        {arriveTangent=-1.087, interpMode=RCIM_Cubic, leaveTangent=-1.0164, tangentMode=RCTM_Break, time=0.3702, value=1}
        <at 1 0>
       ]}
      BlueCurve {keys [<at 0.6491 0.3>,<at 1 0.45>]}
      GreenCurve {keys [<at 0.6491 0.3>,<at 1 0.45>]}
      RedCurve {keys [<at 0.6491 0.3>,<at 1 0.45>]}
     }}
    Scale Mode=::NewEnumerator2
   )
   &DynamicMaterialParameters(
    Index 0 Param 1=100
    Index 0 Param 4=&FloatFromCurve(
     FloatCurve=DataInterface() {propertyValues {Curve {keys [
         {interpMode=RCIM_Cubic, tangentMode=RCTM_User, time=0, value=0}
         {arriveTangent=5.3975, interpMode=RCIM_Cubic, leaveTangent=5.3975, tangentMode=RCTM_User, time=0.4, value=1}
        ]}}}
     Scale Curve=0.25
    )
    Index 1 Param 1=&FloatFromCurve(
     FloatCurve=DataInterface() {propertyValues {Curve {keys [
         {interpMode=RCIM_Cubic, tangentMode=RCTM_User, time=0, value=0}
         {arriveTangent=3.3352, interpMode=RCIM_Cubic, leaveTangent=3.3352, tangentMode=RCTM_User, time=1, value=1}
        ]}}}
     Scale Curve=0.25
    )
    Index 1 Param 2=&FloatFromCurve(
     FloatCurve=DataInterface() {propertyValues {Curve {keys [
         {arriveTangent=-832.0372, interpMode=RCIM_Cubic, leaveTangent=-832.0372, tangentMode=RCTM_User, time=0.2, value=50}
         {arriveTangent=-0.3869, interpMode=RCIM_Cubic, leaveTangent=-0.3869, tangentMode=RCTM_User, time=0.4, value=0}
        ]}}}
     Scale Curve=&RandomRangeFloat(Minimum=0.15, Maximum=0.3)
    )
    Index 1 Param 3=&FloatFromCurve(
     FloatCurve=DataInterface() {propertyValues {Curve {keys [
         {interpMode=RCIM_Cubic, tangentMode=RCTM_User, time=0.2, value=0.5}
         {arriveTangent=2.16, interpMode=RCIM_Cubic, leaveTangent=2.16, tangentMode=RCTM_User, time=1, value=1}
        ]}}}
     Scale Curve=2
    )
    Index 1 Param 4=&FloatFromCurve(FloatCurve=DataInterface() {propertyValues {Curve {keys [<at 0 1>,<at 1 0>]}}}, Scale Curve=-0.25)
    Module.Param0WriteEnabled=true
    Module.Param0WriteEnabled1=true
    Module.Param1WriteEnabled=false
    Module.Param1WriteEnabled1=true
    Module.Param2WriteEnabled=false
    Module.Param2WriteEnabled1=true
    Module.Param3WriteEnabled=true
    Module.Param3WriteEnabled1=true
    Write Parameter Index 1=true
   )
  ]
  rendererClasses [&NiagaraSpriteRendererProperties]
  renderers [{rendererClass=&NiagaraSpriteRendererProperties}]
 }}
summary {
 systemName=NS_Emote_BankNotes_Burst_Flame_Feet_Streaks
 emitters [{emitterName=Streaks_Fat, rendererClasses [&NiagaraSpriteRendererProperties]}]
 userVariables [
  User.bIsFrontEnd {type=NiagaraBool}
  User.bIsFrontEndPreview {type=NiagaraBool}
  User.Global Scale {type=NiagaraFloat, defaultValue=0.075, description=""}
  User.uCameraOffset {type=NiagaraFloat, defaultValue=-5, description=""}
  User.uDirectionInverse {type=NiagaraFloat, defaultValue=1, description=""}
  User.uDurationScale {type=NiagaraFloat, defaultValue=0.75, description=""}
  User.uIndexBoneFoot {type=NiagaraInt32, defaultValue=1}
  User.uOffsetPosition {type=Vector3f, description="", defaultValue {struct=Vector3f, value=<0 0 -15>}}
 ]
}
